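# Looping ambient cue for the Rust Bucket nitro fire stand prop, built from propane torch fire loops.
# A random node picks one of three equally weighted looping waves and feeds it through a modulator for per-play variation.
# Two nested, endlessly looping envelopers of different lengths then drift volume and pitch separately.
# Logarithmic attenuation with listener focus, a concurrency cap of two and a reverb send are set on the cue.
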SoundCue /Game/Sounds/Helios/Environment_Actors/RustBucketTorchFireStand/Spot_RustBucket_Nitro_Fire_Stand_Cue
root=Enveloper_1
nodes [
 {
  id=Enveloper_0
  children [Modulator_0]
  properties {
   bLoop=true
   bLoopIndefinitely=true
   loopEnd=20
   pitchCurve {editorCurveData {keys [{time=0}]}}
   volumeCurve {editorCurveData {keys [<at 0 1 RCIM_Cubic>,<at 10 0.9 RCIM_Cubic>,<at 20 1>]}}
  }
 }
 {
  id=Enveloper_1
  children [Enveloper_0]
  properties {
   bLoop=true
   bLoopIndefinitely=true
   loopEnd=30
   pitchCurve {editorCurveData {keys [{interpMode=RCIM_Cubic, time=0}, <at 15 0.9 RCIM_Cubic>, {time=30}]}}
   volumeCurve {editorCurveData {keys [<at 0 1>]}}
  }
 }
 {id=Modulator_0, children [Random_1]}
 {id=Random_1, children [WavePlayer_6,WavePlayer_7,WavePlayer_8], properties {weights [1,1,1]}}
 {
  id=WavePlayer_6
  properties {
   bLooping=true
   soundWaveAssetPtr=&/Spot_RustBucket_Propane_Torch_Fire_Stand_Loop_01_Cue
  }
 }
 {
  id=WavePlayer_7
  properties {
   bLooping=true
   soundWaveAssetPtr=&/Spot_RustBucket_Propane_Torch_Fire_Stand_Loop_02_Cue
  }
 }
 {
  id=WavePlayer_8
  properties {
   bLooping=true
   soundWaveAssetPtr=&/Spot_RustBucket_Propane_Torch_Fire_Stand_Loop_03_Cue
  }
 }
]
properties {
 bOverrideAttenuation=true
 bOverrideConcurrency=true
 maxDistance=1750
 soundClassObject=&Fort_Ambient_Props_Athena
 volumeMultiplier=0.125
 attenuationOverrides {
  bEnableListenerFocus=true
  distanceAlgorithm=Logarithmic
  falloffDistance=1500
  nonFocusVolumeAttenuation=0.4
  attenuationShapeExtents {x=250}
 }
 concurrencyOverrides {maxCount=2}
 soundSubmixSends [{minSendLevel=0, soundSubmix=&Fortnite_Reverb_Default}]
}
