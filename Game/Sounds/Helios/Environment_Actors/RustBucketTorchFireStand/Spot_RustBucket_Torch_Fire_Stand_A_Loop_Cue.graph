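# Looping ambient fire cue for the rust bucket torch fire stand prop, classed as Athena ambient prop audio.
# A random pick between two mixes (oil fire plus propane torch, or two torch fire loops) feeds a root pitch modulator.
# Both mixes share a random layer of three propane loop variants.
# Overrides add logarithmic falloff, listener focus and a tight voice limit, with sends to exterior ambient bus and reverb.
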SoundCue /Game/Sounds/Helios/Environment_Actors/RustBucketTorchFireStand/Spot_RustBucket_Torch_Fire_Stand_A_Loop_Cue
root=Modulator_1
nodes [
 {id=Mixer_0, children [WavePlayer_3,WavePlayer_0,Random_1], properties {inputVolume [0.4,3,2]}}
 {id=Mixer_1, children [WavePlayer_1,WavePlayer_2,Random_1], properties {inputVolume [1,3,2]}}
 {id=Modulator_1, children [Random_0], properties {pitchMax=1.1, pitchMin=0.9}}
 {id=Random_0, children [Mixer_0,Mixer_1], properties {weights [1,1]}}
 {id=Random_1, children [WavePlayer_4,WavePlayer_5,WavePlayer_6], properties {weights [1,1,1]}}
 {
  id=WavePlayer_0
  properties {bLooping=true, soundWaveAssetPtr=/Game/Sounds/Helios/Environment_Actors/Brazier/Spot_Statue_Oil_Fire_Small_Loop}
 }
 {id=WavePlayer_1, properties {bLooping=true, soundWaveAssetPtr=/Game/Sounds/Trans_AmbientLoops/Fire/Fire_Torch_01}}
 {id=WavePlayer_2, properties {bLooping=true, soundWaveAssetPtr=/Game/Sounds/Creative/Gadgets/Torch/Torch_Fire_Loop01}}
 {
  id=WavePlayer_3
  properties {
   bLooping=true
   soundWaveAssetPtr=&/Spot_RustBucket_Propane_Torch_Fire_Stand_Loop_Cue
  }
 }
 {
  id=WavePlayer_4
  properties {
   bLooping=true
   soundWaveAssetPtr=&/Spot_RustBucket_Propane_Torch_Fire_Stand_Loop_03_Cue
  }
 }
 {
  id=WavePlayer_5
  properties {
   bLooping=true
   soundWaveAssetPtr=&/Spot_RustBucket_Propane_Torch_Fire_Stand_Loop_02_Cue
  }
 }
 {
  id=WavePlayer_6
  properties {
   bLooping=true
   soundWaveAssetPtr=&/Spot_RustBucket_Propane_Torch_Fire_Stand_Loop_01_Cue
  }
 }
]
properties {
 attenuationSettings=&/Spot_RustBucket_Torch_Fire_Stand_Att
 bBypassVolumeScaleForPriority=true
 bOverrideAttenuation=true
 bOverrideConcurrency=true
 maxDistance=1750
 priority=0
 soundClassObject=&Fort_Ambient_Props_Athena
 volumeMultiplier=0.125
 attenuationOverrides {
  bEnableFocusInterpolation=true
  bEnableListenerFocus=true
  distanceAlgorithm=Logarithmic
  falloffDistance=1500
  focusAttackInterpSpeed=0.9
  focusAzimuth=15
  focusReleaseInterpSpeed=0.4
  nonFocusAzimuth=45
  nonFocusVolumeAttenuation=0.1
  stereoSpread=10
  attenuationShapeExtents {x=250}
 }
 busSends [{audioBus=null, minSendLevel=1, sendLevel=1, soundSourceBus=&AmbientAudio_Exterior_SourceBus}]
 concurrencyOverrides {maxCount=2, voiceStealReleaseTime=2}
 soundSubmixSends [{minSendLevel=0.6, soundSubmix=&Fortnite_Reverb_Default}]
}
