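# Looping ambient cue for the Rust Bucket torch fire stand, playing wind-driven metal rattles over a lowered-pitch space heater hum.
# A random pick between a wire-wind loop and two mixed metal-vibration loops gets random pitch.
# Chained looping envelopers add slow pitch drift and two-minute volume swells, behind a short fade-in.
# Only the first two quality levels play sound, with capsule attenuation, listener focus and an exterior ambience bus send.
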SoundCue /Game/Sounds/Helios/Environment_Actors/RustBucketTorchFireStand/Spot_RustBucket_Torch_Fire_Stand_Metal_Wind_Rattles_Loop_Cue
root=QualityLevel_0
nodes [
 {
  id=Enveloper_1
  children [Modulator_1]
  properties {
   bLoop=true
   bLoopIndefinitely=true
   loopEnd=30
   pitchCurve {editorCurveData {keys [{interpMode=RCIM_Cubic, time=0}, <at 15.9468 0.9 RCIM_Cubic>, {interpMode=RCIM_Cubic, time=30}]}}
   volumeCurve {editorCurveData {keys [<at 0 1>]}}
  }
 }
 {
  id=Enveloper_2
  children [Enveloper_1]
  properties {
   bLoop=true
   bLoopIndefinitely=true
   loopEnd=120
   pitchCurve {editorCurveData {keys [{time=0}]}}
   volumeCurve {editorCurveData {keys [
      {arriveTangent=0.0001, interpMode=RCIM_Cubic, leaveTangent=0.0082, tangentMode=RCTM_User, time=0.4639, value=0.3721}
      {arriveTangent=0.0321, interpMode=RCIM_Cubic, leaveTangent=0.0321, time=10, value=0.6}
      {arriveTangent=-0.003, interpMode=RCIM_Cubic, leaveTangent=-0.003, time=20, value=1}
      <at 45.7834 0.4918 RCIM_Cubic>
      {arriveTangent=0.0153, interpMode=RCIM_Cubic, leaveTangent=0.0153, time=60, value=1}
      {arriveTangent=-0.0202, interpMode=RCIM_Cubic, leaveTangent=-0.0202, tangentMode=RCTM_User, time=82.1697, value=1.05}
      <at 119.9639 0.4918>
     ]}}
  }
 }
 {
  id=Enveloper_29
  children [Enveloper_2]
  properties {
   pitchCurve {editorCurveData {keys [{time=0}]}}
   volumeCurve {editorCurveData {keys [{arriveTangent=3.4293, interpMode=RCIM_Cubic, leaveTangent=3.4293, tangentMode=RCTM_User, time=0, value=0}, <at 0.4 1 RCIM_Cubic>]}}
  }
 }
 {id=Mixer_1, children [WavePlayer_1,WavePlayer_3], properties {inputVolume [1.7,1.2]}}
 {id=Mixer_2, children [WavePlayer_2], properties {inputVolume [1.5]}}
 {id=Mixer_3, children [Enveloper_29,Modulator_0], properties {inputVolume [1,3.5]}}
 {id=Modulator_0, children [WavePlayer_7], properties {pitchMax=0.9, pitchMin=0.9}}
 {id=Modulator_1, children [Random_0], properties {pitchMax=1.2, pitchMin=1}}
 {id=QualityLevel_0, children [Mixer_3,Mixer_3,null,null]}
 {id=Random_0, children [Mixer_2,Mixer_1], properties {weights [1,1]}}
 {
  id=WavePlayer_1
  properties {bLooping=true, soundWaveAssetPtr=/Game/Sounds/Artemis/POI/Blimps_Crashsites/Spots/Spot_Wind_Excited_Metal_Vibrating_Loop}
 }
 {id=WavePlayer_2, properties {bLooping=true, soundWaveAssetPtr=/Game/Sounds/Asteria/POI/SkyIsland/Spots/Spot_Ascender_Ambient_Wire_Wind_Loop}}
 {
  id=WavePlayer_3
  properties {bLooping=true, soundWaveAssetPtr=/Game/Sounds/Artemis/POI/Hollow_Base/Spots/Spot_Wind_Excited_Metal_Vibrating_Loop}
 }
 {
  id=WavePlayer_7
  properties {bLooping=true, soundWaveAssetPtr=/Game/Sounds/Asteria/Environment_Actors/GiantSpaceHeater/Spot_GiantSpaceHeater_Hum_Loop}
 }
]
properties {
 attenuationSettings=/Game/Sounds/Helios/POI/MountainRescueStation/Spots/Spot_Exterior_Antenna_Wind_Metal_Rattle_Att
 bOverrideAttenuation=true
 bOverrideConcurrency=true
 maxDistance=3750
 soundClassObject=&Fort_Ambient_Props_Athena
 volumeMultiplier=0.501
 attenuationOverrides {
  attenuationShape=Capsule
  bEnableListenerFocus=true
  falloffDistance=3000
  focusAzimuth=25
  nonFocusAzimuth=45
  nonFocusVolumeAttenuation=0.2
  stereoSpread=10
  attenuationShapeExtents=<750 350>
 }
 busSends [{audioBus=null, minSendLevel=1, sendLevel=1, soundSourceBus=&AmbientAudio_Exterior_SourceBus}]
 concurrencyOverrides {resolutionRule=PreventNew}
}
